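# Sound cue for the electric ceiling trap's charged state, playing a continuous loop of randomly chosen mono spark sounds.
# A quality-level root selects between two looping random branches, one with eight variations and one with a reduced set of five, leaving two slots empty.
# In both branches one spark wave is pitched down by a modulator.
# Fort trap loop attenuation applies, and a concurrency override caps simultaneous instances, stopping the quietest.
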SoundCue /Figment_Trap_Ceiling_Electric_Athena/Sounds/Fort_Traps/Electric_Trap/Electric_Trap_Charged
root=QualityLevel_0
nodes [
 {id=Looping_1, children [Random_1]}
 {id=Looping_2, children [Random_2]}
 {id=Modulator_1, children [WavePlayer_9], properties {pitchMax=0.84, pitchMin=0.84}}
 {id=Modulator_2, children [WavePlayer_17], properties {pitchMax=0.84, pitchMin=0.84}}
 {id=QualityLevel_0, children [Looping_1,Looping_2,null,null]}
 {
  id=Random_1
  children [WavePlayer_8, Modulator_1, WavePlayer_10, WavePlayer_11, WavePlayer_12, WavePlayer_13, WavePlayer_14, WavePlayer_15]
  properties {weights [1,1,1,1,1,1,1,1]}
 }
 {id=Random_2, children [WavePlayer_16, Modulator_2, WavePlayer_18, WavePlayer_19, WavePlayer_22], properties {weights [1,1,1,1,1]}}
 {id=WavePlayer_10, properties {soundWaveAssetPtr=/Game/Sounds/Fort_Traps/Electric_Trap/CeilingTrap_Electric_Spark_mono_001}}
 {id=WavePlayer_11, properties {soundWaveAssetPtr=/Game/Sounds/Fort_Traps/Electric_Trap/CeilingTrap_Electric_Spark_mono_002}}
 {id=WavePlayer_12, properties {soundWaveAssetPtr=/Game/Sounds/Fort_Traps/Electric_Trap/CeilingTrap_Electric_Spark_mono_003}}
 {id=WavePlayer_13, properties {soundWaveAssetPtr=/Game/Sounds/Fort_Traps/Electric_Trap/CeilingTrap_Electric_Spark_mono_005}}
 {id=WavePlayer_14, properties {soundWaveAssetPtr=/Game/Sounds/Fort_Traps/Electric_Trap/CeilingTrap_Electric_Spark_mono_006}}
 {id=WavePlayer_15, properties {soundWaveAssetPtr=/Game/Sounds/Fort_Traps/Electric_Trap/CeilingTrap_Electric_Spark_mono_007}}
 {id=WavePlayer_16, properties {soundWaveAssetPtr=/Game/Sounds/Fort_Traps/Electric_Trap/CeilingTrap_Electric_Spark_mono_004}}
 {
  id=WavePlayer_17
  properties {soundWaveAssetPtr=&/CeilingTrap_Electric_Spark_mono_008}
 }
 {id=WavePlayer_18, properties {soundWaveAssetPtr=&CeilingTrap_Electric_Spark_mono_001}}
 {id=WavePlayer_19, properties {soundWaveAssetPtr=&CeilingTrap_Electric_Spark_mono_002}}
 {id=WavePlayer_22, properties {soundWaveAssetPtr=&CeilingTrap_Electric_Spark_mono_006}}
 {id=WavePlayer_8, properties {soundWaveAssetPtr=&CeilingTrap_Electric_Spark_mono_004}}
 {
  id=WavePlayer_9
  properties {soundWaveAssetPtr=&CeilingTrap_Electric_Spark_mono_008}
 }
]
properties {
 attenuationSettings=/Game/Sounds/Attenuation/Fort_Traps_Loop_Att
 bOverrideConcurrency=true
 maxDistance=2500
 soundClassObject=&Fort_Traps
 volumeMultiplier=0.5
 attenuationOverrides {falloffDistance=2100}
 concurrencyOverrides {maxCount=2, resolutionRule=StopQuietest}
}
