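# Sound cue for the destruction explosion of the electric ceiling trap, using the Fort_Traps sound class and attenuation.
# A quality-level root sends its first two levels to an equally weighted random pick of three destroyed waves, and its last two to the first wave alone.
# Each branch passes through a default Modulator for slight pitch and volume variation.
# The cue plays at reduced volume, and its concurrency override prevents new instances once the limit is reached.
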SoundCue /Figment_Trap_Ceiling_Electric_Athena/Sounds/Fort_Traps/Electric_Trap/Electric_Trap_Explode
root=QualityLevel_0
nodes [
 {id=Modulator_0, children [WavePlayer_0]}
 {id=Modulator_4, children [Random_3]}
 {id=QualityLevel_0, children [Modulator_4,Modulator_4,Modulator_0,Modulator_0]}
 {id=Random_3, children [WavePlayer_14, WavePlayer_15, WavePlayer_16], properties {weights [1,1,1]}}
 {
  id=WavePlayer_0
  properties {soundWaveAssetPtr=&/CeilingTrap_Electric_Destroyed_001}
 }
 {
  id=WavePlayer_14
  properties {soundWaveAssetPtr=&CeilingTrap_Electric_Destroyed_001}
 }
 {
  id=WavePlayer_15
  properties {soundWaveAssetPtr=&/CeilingTrap_Electric_Destroyed_002}
 }
 {
  id=WavePlayer_16
  properties {soundWaveAssetPtr=&/CeilingTrap_Electric_Destroyed_003}
 }
]
properties {
 attenuationSettings=&Fort_Traps_Att
 bOverrideConcurrency=true
 maxDistance=6000
 soundClassObject=&Fort_Traps
 volumeMultiplier=0.5
 concurrencyOverrides {resolutionRule=PreventNew}
}
