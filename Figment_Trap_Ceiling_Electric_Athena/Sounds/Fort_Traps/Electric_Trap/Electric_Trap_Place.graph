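# Sound cue played when the ceiling electric trap is placed, routed through the trap placement sound class.
# A mixer layers the trap's electric ready sound with a quieter construction-done sound.
# A quality-level switch gives two levels a random pick of two ready waves and the other two a single optimized wave.
# Shared trap attenuation sets falloff, and a concurrency override caps simultaneous instances.
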SoundCue /Figment_Trap_Ceiling_Electric_Athena/Sounds/Fort_Traps/Electric_Trap/Electric_Trap_Place
root=Mixer_8
nodes [
 {id=Mixer_8, children [QualityLevel_0, WavePlayer_83], properties {inputVolume [1,0.7]}}
 {id=QualityLevel_0, children [Random_5,Random_5,WavePlayer_0,WavePlayer_0]}
 {id=Random_5, children [WavePlayer_19, WavePlayer_43], properties {weights [1,1]}}
 {id=WavePlayer_0, properties {soundWaveAssetPtr=/FNGameplayCues/Sounds/Fort_Traps/Electric_Trap/CeilingTrap_Electric_Ready_002_opt}}
 {id=WavePlayer_19, properties {soundWaveAssetPtr=/Game/Sounds/Fort_Traps/Electric_Trap/CeilingTrap_Electric_Ready_001}}
 {id=WavePlayer_43, properties {soundWaveAssetPtr=/FNGameplayCues/Sounds/Fort_Traps/Electric_Trap/CeilingTrap_Electric_Ready_002}}
 {id=WavePlayer_83, properties {soundWaveAssetPtr=/Game/Sounds/Fort_UI_Sounds/Gameplay/Fort_Contructino_Done_02}}
]
properties {
 attenuationSettings=&Fort_Traps_Att
 bOverrideConcurrency=true
 maxDistance=6000
 soundClassObject=/Game/Sounds/FortSoundClassesAndModes/Fort_Traps_Placement
 volumeMultiplier=0.9
 attenuationOverrides {falloffDistance=2400, attenuationShapeExtents {x=600}}
 concurrencyOverrides {maxCount=3}
}
