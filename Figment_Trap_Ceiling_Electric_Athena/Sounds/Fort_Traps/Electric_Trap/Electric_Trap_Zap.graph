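# Sound cue for the electric ceiling trap zap, pairing a fire sound with a ready sound delayed 0.7 seconds.
# A quality-level root picks one of four mixes: the first randomizes among fire and ready variants, while the other three use single waves with slight fire-pitch modulation.
# The last two tiers use an "opt" ready wave.
# Trap attenuation has overridden falloff, and concurrency prevents new instances at the limit.
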SoundCue /Figment_Trap_Ceiling_Electric_Athena/Sounds/Fort_Traps/Electric_Trap/Electric_Trap_Zap
root=QualityLevel_0
nodes [
 {id=Delay_1, children [WavePlayer_9], properties {delayMax=0.7, delayMin=0.7}}
 {id=Delay_2, children [WavePlayer_0], properties {delayMax=0.7, delayMin=0.7}}
 {id=Delay_3, children [Random_10], properties {delayMax=0.7, delayMin=0.7}}
 {id=Delay_4, children [WavePlayer_1], properties {delayMax=0.7, delayMin=0.7}}
 {id=Mixer_1, children [Modulator_0,Delay_1], properties {inputVolume [1,1]}}
 {id=Mixer_18, children [Random_1,Delay_3], properties {inputVolume [1,1]}}
 {id=Mixer_2, children [Modulator_4,Delay_2], properties {inputVolume [1,1]}}
 {id=Mixer_3, children [Modulator_5,Delay_4], properties {inputVolume [1,1]}}
 {id=Modulator_0, children [WavePlayer_8], properties {pitchMax=1.1, pitchMin=1}}
 {id=Modulator_4, children [WavePlayer_10], properties {pitchMax=1.1, pitchMin=1}}
 {id=Modulator_5, children [WavePlayer_12], properties {pitchMax=1.1, pitchMin=1}}
 {id=QualityLevel_0, children [Mixer_18,Mixer_1,Mixer_2,Mixer_3]}
 {id=Random_1, children [WavePlayer_3,WavePlayer_4,WavePlayer_5], properties {weights [1,1,1]}}
 {id=Random_10, children [WavePlayer_38, WavePlayer_39], properties {weights [1,1]}}
 {id=WavePlayer_0, properties {soundWaveAssetPtr=/FNGameplayCues/Sounds/Fort_Traps/Electric_Trap/CeilingTrap_Electric_Ready_002_opt}}
 {id=WavePlayer_1, properties {soundWaveAssetPtr=&CeilingTrap_Electric_Ready_002_opt}}
 {id=WavePlayer_10, properties {soundWaveAssetPtr=/Game/Sounds/Fort_Traps/Electric_Trap/CeilingTrap_Electric_Fire_002}}
 {id=WavePlayer_12, properties {soundWaveAssetPtr=&CeilingTrap_Electric_Fire_002}}
 {id=WavePlayer_3, properties {soundWaveAssetPtr=/FNGameplayCues/Sounds/Fort_Traps/Electric_Trap/CeilingTrap_Electric_Fire_001}}
 {id=WavePlayer_38, properties {soundWaveAssetPtr=/Game/Sounds/Fort_Traps/Electric_Trap/CeilingTrap_Electric_Ready_001}}
 {id=WavePlayer_39, properties {soundWaveAssetPtr=/FNGameplayCues/Sounds/Fort_Traps/Electric_Trap/CeilingTrap_Electric_Ready_002}}
 {id=WavePlayer_4, properties {soundWaveAssetPtr=&CeilingTrap_Electric_Fire_002}}
 {id=WavePlayer_5, properties {soundWaveAssetPtr=/FNGameplayCues/Sounds/Fort_Traps/Electric_Trap/CeilingTrap_Electric_Fire_004}}
 {id=WavePlayer_8, properties {soundWaveAssetPtr=&CeilingTrap_Electric_Fire_002}}
 {id=WavePlayer_9, properties {soundWaveAssetPtr=&CeilingTrap_Electric_Ready_002}}
]
properties {
 attenuationSettings=&Fort_Traps_Att
 bOverrideConcurrency=true
 maxDistance=6000
 soundClassObject=&Fort_Traps
 volumeMultiplier=0.5
 attenuationOverrides {falloffDistance=3200, attenuationShapeExtents {x=800}}
 concurrencyOverrides {resolutionRule=PreventNew}
}
